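; Season 23 narrative Part 02 parent quest actor that customizes per-player visuals and interaction when its quest is not ready.
; OnQuestNotReady branches on test-visual, visible-on-complete and visible-to-all flags, prior interactors and objective completion.
; Visible-to-all actors show pre- or post-interaction meshes; otherwise, with zero state-change delay, it can hide them, stop distance watching and block interaction.
; Construction defers to the parent; BeginPlay, overlap and tick add no logic.
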(event EventBeginPlay
  (|CallParentFunction) ; This node is disabled and will not be called.\nDrag off pins to build functionality.
)

(event Collision|EventActorBeginOverlap
  (params (OtherActor :type Actor%20Object%20Reference)))

(event EventTick
  (params (DeltaSeconds :type Float%20%28single-precision%29)))

(event OnQuestNotReady(ConsolidatedQuestComponent)
  (params (PlayerController :type Fort%20Player%20Controller%20Object%20Reference))
  ;[1 1 1] OnQuestNotReady(QuestComponent)
  (if (Variables|Design|QuestSetup|GetbTestVisualsNoQuestRequired)
    _
    (elif (Variables|Default|GetVisibleOnQuestComplete)
      ;;[1 1 1] Is That The Match When Player Completed the Quest?
      (if (Utilities|Array|ContainsItem (Variables|Advanced|Interaction|GetFortPlayerControllersWhoInteracted) PlayerController)
        ;;[1 1 1] If so, don't hide the actors player interacted with
        (if (ObjectiveState|HasCompletedObjective PlayerController)
          _
          (else
            _))
        (else
          _))
      (if (Variables|Design|QuestSetup|GetbVisibleToAll)
        (if (ObjectiveState|HasCompletedObjective PlayerController)
          ;[1 1 1] If player completed the quest, show the post-interaction version
          (VisualState|Meshes|PostInteractionVisuals PlayerController)
          (else
            (VisualState|Meshes|PreInteractionVisuals PlayerController)))
        (elif (== (Variables|Design|PostInteraction|GetStateChangeDelay) 0)
          (VisualState|Meshes|HiddenVisuals PlayerController)
          (if (Variables|Advanced|IntervalCheck|GetbUsingDistanceCheck)
            (CallFunction|EventRemoveWatchedActor PlayerController)
            (else
              _))
          (InteractionState|HandleSetCannotInteract PlayerController)
          (else
            _)))
      (else
        _))))

(fn ConstructionScript
  (|CallParentFunction))
