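; Quest actor for the final satellite in the Season 23 Battle Pass narrative quest (Part 02, Q01), built on a parent interaction actor.
; After the parent's post-interaction VFX and audio event runs, it snaps the post-interaction ambient component to the FX_Idle socket on the post-interaction mesh.
; Only the socket's world location and rotation are applied, leaving scale at default.
; The BeginPlay, overlap and Tick events are disabled stubs.
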(event EventBeginPlay
  (|CallParentFunction) ; This node is disabled and will not be called.\nDrag off pins to build functionality.
)

(event Collision|EventActorBeginOverlap
  (params (OtherActor :type Actor%20Object%20Reference))
  (|CallParentFunction OtherActor) ; This node is disabled and will not be called.\nDrag off pins to build functionality.
)

(event EventTick
  (params (DeltaSeconds :type Float%20%28single-precision%29))
  (|CallParentFunction DeltaSeconds) ; This node is disabled and will not be called.\nDrag off pins to build functionality.
)

(event VisualState|VFX&Audio|EventPostInteractionVFX&Audio
  (params (PlayerController :type Fort%20Player%20Controller%20Object%20Reference))
  (bind (_returnvalue_location _returnvalue_rotation) (Math|Transform|BreakTransform (Transformation|GetSocketTransform (Variables|Components|GetSM_Quest_PostInteraction) "FX_Idle" "RTS_World")))
  (|CallParentFunction PlayerController)
  (Transformation|SetWorldTransform (Variables|Components|GetN_PostInteractionAmbient) (Math|Transform|MakeTransform _returnvalue_location _returnvalue_rotation) false false))

(fn ConstructionScript
  (|CallParentFunction))
