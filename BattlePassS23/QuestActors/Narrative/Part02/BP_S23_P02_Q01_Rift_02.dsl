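; BP_S23_P02_Q01_Rift_02 is a Battle Pass Season 23 narrative quest actor, a rift for Part 02 Quest 01, built on a parent quest-actor Blueprint.
; When the quest becomes ready through its ConsolidatedQuestComponent, it swaps slot 1 of the SM_Quest_PreInteraction mesh to the PreInteraction material override.
; Begin play, overlap and tick events are empty, and the construction script only calls the parent.
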(event EventBeginPlay
  (|CallParentFunction) ; This node is disabled and will not be called.\nDrag off pins to build functionality.
)

(event Collision|EventActorBeginOverlap
  (params (OtherActor :type Actor%20Object%20Reference)))

(event EventTick
  (params (DeltaSeconds :type Float%20%28single-precision%29)))

(event OnQuestReady(ConsolidatedQuestComponent)
  (params (PlayerController :type Fort%20Player%20Controller%20Object%20Reference))
  (Rendering|Material|SetMaterial (Variables|Components|GetSM_Quest_PreInteraction) 1 (Variables|Design|PreInteraction|GetPreInteraction_MIS_Override)))

(fn ConstructionScript
  (|CallParentFunction))
