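; Quest actor Blueprint for a sensor monitor in the Season 23 Battle Pass narrative quests, Part 2 Quest 2.
; When the consolidated quest component signals that the quest is ready, it checks an ApplyGhostMaterial flag.
; If set, it applies the pre-interaction material override to slot 0 of the pre-interaction quest mesh.
; The construction script only calls its parent.
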(event OnQuestReady(ConsolidatedQuestComponent)
  (params (PlayerController :type Fort%20Player%20Controller%20Object%20Reference))
  (if (Variables|Default|GetApplyGhostMaterial)
    (Rendering|Material|SetMaterial (Variables|Components|GetSM_Quest_PreInteraction) 0 (Variables|Design|PreInteraction|GetPreInteraction_MIS_Override))))

(fn ConstructionScript
  (|CallParentFunction))
