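; Quest actor for a Battle Pass S23 sensor monitor's final interaction, animating its signal-tracker screen through materials.
; When the quest is ready, a looping timer cycles the screen mesh through five stage materials and an empty one.
; When not ready, flag-gated logic resets the timer handle, blanks the screen, then either picks pre/post-interaction visuals or, without a state-change delay, hides it and blocks interaction.
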(event EventBeginPlay
  (|CallParentFunction))

(event Collision|EventActorBeginOverlap
  (params (OtherActor :type Actor%20Object%20Reference))
  (|CallParentFunction OtherActor) ; This node is disabled and will not be called.\nDrag off pins to build functionality.
)

(event EventTick
  (params (DeltaSeconds :type Float%20%28single-precision%29))
  (|CallParentFunction DeltaSeconds) ; This node is disabled and will not be called.\nDrag off pins to build functionality.
)

(event OnQuestNotReady(ConsolidatedQuestComponent)
  (params (PlayerController :type Fort%20Player%20Controller%20Object%20Reference))
  ;[1 1 1] OnQuestNotReady(QuestComponent)
  (if (Variables|Design|QuestSetup|GetbTestVisualsNoQuestRequired)
    _
    (elif (Variables|Default|GetVisibleOnQuestComplete)
      ;;[1 1 1] Is That The Match When Player Completed the Quest?
      (if (Utilities|Array|ContainsItem (Variables|Advanced|Interaction|GetFortPlayerControllersWhoInteracted) PlayerController)
        ;;[1 1 1] If so, don't hide the actors player interacted with
        (if (ObjectiveState|HasCompletedObjective PlayerController)
          _
          (else
            _))
        (else
          _))
      (Utilities|FlowControl|Sequence
        (:then_0
          (if (Utilities|Time|IsValidTimerHandle (Variables|Default|GetImageSwitcherTimer))
            (Variables|Default|SetImageSwitcherTimer)
            (Rendering|Material|SetMaterial (Variables|Default|GetSM_Quest_SignalTracker_A_Main_Place_Pre) 0 (Variables|Default|GetEmptyMaterial))))
        (:then_1
          (if (Variables|Design|QuestSetup|GetbVisibleToAll)
            (if (ObjectiveState|HasCompletedObjective PlayerController)
              ;[1 1 1] If player completed the quest, show the post-interaction version
              (VisualState|Meshes|PostInteractionVisuals PlayerController)
              (else
                (VisualState|Meshes|PreInteractionVisuals PlayerController)))
            (elif (== (Variables|Design|PostInteraction|GetStateChangeDelay) 0)
              (VisualState|Meshes|HiddenVisuals PlayerController)
              (if (Variables|Advanced|IntervalCheck|GetbUsingDistanceCheck)
                (CallFunction|EventRemoveWatchedActor PlayerController)
                (else
                  _))
              (InteractionState|HandleSetCannotInteract PlayerController)
              (else
                _)))))
      (else
        _))))

(event OnQuestReady(ConsolidatedQuestComponent)
  (params (PlayerController :type Fort%20Player%20Controller%20Object%20Reference))
  (bind _returnvalue (Utilities|Time|SetTimerbyEvent (delegate Custom|UpdateImage) 0.4 true))
  (Variables|Default|SetImageSwitcherTimer _returnvalue))

(event Custom|UpdateImage
  (bind _selection (Variables|Default|GetSelection))
  (switch int (Variables|Default|GetSelection)
    (:0
      (Variables|Default|SetMaterial (Variables|Default|GetStage1Material)))
    (:1
      (Variables|Default|SetMaterial (Variables|Default|GetStage2Material)))
    (:2
      (Variables|Default|SetMaterial (Variables|Default|GetStage3Material)))
    (:3
      (Variables|Default|SetMaterial (Variables|Default|GetStage4Material)))
    (:4
      (Variables|Default|SetMaterial (Variables|Default|GetStage5Material)))
    (:5
      (Variables|Default|SetMaterial (Variables|Default|GetEmptyMaterial))))
  (Rendering|Material|SetMaterial (Variables|Default|GetSM_Quest_SignalTracker_A_Main_Place_Pre) 0 (Variables|Default|GetMaterial))
  (Math|Integer|IncrementInt _selection
    (:%20%20
      (Math|Integer|CompareInt _selection 5
        (:>
          (Variables|Default|SetSelection 0))))))

(fn ConstructionScript
  (|CallParentFunction))
