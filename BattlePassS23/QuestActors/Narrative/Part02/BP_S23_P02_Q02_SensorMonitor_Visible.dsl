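; Battle Pass Season 23 quest actor for a sensor monitor (signal tracker) prop whose screen changes as its narrative quest progresses.
; On each quest update from the consolidated quest component, it picks the stage 2 to 5 material from a Selection counter.
; It applies that material to slot 0 of the signal tracker mesh, then increments the counter, so stages advance by update count.
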(event EventBeginPlay
  (|CallParentFunction) ; This node is disabled and will not be called.\nDrag off pins to build functionality.
)

(event Collision|EventActorBeginOverlap
  (params (OtherActor :type Actor%20Object%20Reference))
  (|CallParentFunction OtherActor) ; This node is disabled and will not be called.\nDrag off pins to build functionality.
)

(event EventTick
  (params (DeltaSeconds :type Float%20%28single-precision%29))
  (|CallParentFunction DeltaSeconds) ; This node is disabled and will not be called.\nDrag off pins to build functionality.
)

(event OnUpdateQuest(ConsolidatedQuestComponent)
  (params (PlayerController :type Fort%20Player%20Controller%20Object%20Reference))
  (switch int (Variables|Default|GetSelection)
    (:0
      (Variables|Default|SetMaterial (Variables|Default|GetStage2Material)))
    (:1
      (Variables|Default|SetMaterial (Variables|Default|GetStage3Material)))
    (:2
      (Variables|Default|SetMaterial (Variables|Default|GetStage4Material)))
    (:3
      (Variables|Default|SetMaterial (Variables|Default|GetStage5Material))))
  (Rendering|Material|SetMaterial (Variables|Default|GetSM_Quest_SignalTracker_A_Main_Place_Pre) 0 (Variables|Default|GetMaterial))
  (Math|Integer|IncrementInt (Variables|Default|GetSelection)))

(fn ConstructionScript
  (|CallParentFunction))
