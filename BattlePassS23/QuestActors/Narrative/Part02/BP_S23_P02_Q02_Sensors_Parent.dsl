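; Parent quest actor blueprint for sensor props in Battle Pass S23 narrative quest Part 02, Quest 02.
; When the quest is ready, an optional flag swaps the pre-interaction mesh's "SignalTracker" slot to a ghostly material and sets material index 1 to a design override.
; Post-interaction visuals run parent logic, then hide the pre-interaction mesh for that player controller.
; Begin play, overlap and tick events hold no active logic.
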(event EventBeginPlay
  (|CallParentFunction) ; This node is disabled and will not be called.\nDrag off pins to build functionality.
)

(event Collision|EventActorBeginOverlap
  (params (OtherActor :type Actor%20Object%20Reference)))

(event EventTick
  (params (DeltaSeconds :type Float%20%28single-precision%29)))

(event OnQuestReady(ConsolidatedQuestComponent)
  (params (PlayerController :type Fort%20Player%20Controller%20Object%20Reference))
  (if (Variables|Default|GetApplyGhostlyMaterial)
    (Rendering|Material|SetMaterialbyName (Variables|Components|GetSM_Quest_PreInteraction) "SignalTracker" "/Game/Athena/Items/QuestParents/PropQuestActor/MI_QuestPropGhostly_02_Inst.MI_QuestPropGhostly_02_Inst")
    (Rendering|Material|SetMaterial (Variables|Components|GetSM_Quest_PreInteraction) 1 (Variables|Design|PreInteraction|GetPreInteraction_MIS_Override))))

(event VisualState|Meshes|EventPostInteractionVisuals
  (params (PlayerController :type Fort%20Player%20Controller%20Object%20Reference))
  (|CallParentFunction PlayerController)
  (VisualState|Meshes|SetMeshHidden (Variables|Components|GetSM_Quest_PreInteraction) PlayerController true))

(fn ConstructionScript
  (|CallParentFunction))
