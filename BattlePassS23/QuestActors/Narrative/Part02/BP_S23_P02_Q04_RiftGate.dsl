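; Battle Pass Season 23 quest rift gate actor that animates its ambient rift effect when the quest updates.
; A looping interval timer feeds the duration timer's normalized progress into the Niagara "IntoRift" float.
; When the duration ends, the same end event re-arms with a reverse duration and delay, inverting progress to ramp back down.
; An active flag blocks retriggering until cleanup clears both timers, which also runs on EndPlay.
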(event EventBeginPlay
  (|CallParentFunction) ; This node is disabled and will not be called.\nDrag off pins to build functionality.
)

(event Collision|EventActorBeginOverlap
  (params (OtherActor :type Actor%20Object%20Reference)))

(event EventTick
  (params (DeltaSeconds :type Float%20%28single-precision%29)))

(event OnUpdateQuest(ConsolidatedQuestComponent)
  (params (PlayerController :type Fort%20Player%20Controller%20Object%20Reference))
  (if (not (Variables|Default|GetVfxActive))
    (Variables|Default|SetVfxActive true)
    (Variables|Default|SetCurrentRiftDuration (Variables|Default|GetIntoRiftDuration))
    (bind _returnvalue (Utilities|Time|SetTimerbyEvent (EventDispatchers|CreateEvent :!event Custom|UpdateIntoRiftVfx) (Variables|Default|GetIntoRiftInterval) true))
    (Variables|Default|SetIntoRiftIntervalTimer _returnvalue)
    (bind _returnvalue_2 (Utilities|Time|SetTimerbyEvent (EventDispatchers|CreateEvent :!event Custom|EndIntoRiftVfx) (Variables|Default|GetIntoRiftDuration)))
    (Variables|Default|SetIntoRiftDurationTimer _returnvalue_2)))

(event VisualState|VFX&Audio|EventPostInteractionVFX&Audio
  (params (PlayerController :type Fort%20Player%20Controller%20Object%20Reference))
  (|CallParentFunction PlayerController))

(event Custom|UpdateIntoRiftVfx
  (bind _currentriftduration (Variables|Default|GetCurrentRiftDuration))
  (bind _intoriftdurationtimer (Variables|Default|GetIntoRiftDurationTimer))
  (bind _returnvalue (Utilities|Time|GetTimerElapsedTimebyHandle _intoriftdurationtimer))
  (bind _returnvalue_2 (Math|Float|SafeDivide _returnvalue _currentriftduration))
  (if (Variables|Default|GetIsInverted)
    (Variables|Default|SetCurrentUserVariable (- 1.0 _returnvalue_2))
    (else
      (Variables|Default|SetCurrentUserVariable _returnvalue_2)))
  (Niagara|SetNiagaraVariable(Float) (Variables|Default|GetRift_Ambient) "IntoRift" (Math|Float|Clamp(Float) (Variables|Default|GetCurrentUserVariable))))

(event Custom|EndIntoRiftVfx
  (if (Variables|Default|GetIsInverted)
    (CallFunction|CleanupTimers)
    (else
      (Niagara|SetNiagaraVariable(Float) (Variables|Default|GetRift_Ambient) "IntoRift" 1.0)
      (bind _returnvalue (Utilities|Time|SetTimerbyEvent (EventDispatchers|CreateEvent :!event Custom|EndIntoRiftVfx) (Variables|Default|GetIntoRiftReverseDuration) false false (Variables|Default|GetIntoRiftReverseDelay)))
      (Variables|Default|SetIntoRiftDurationTimer _returnvalue)
      (Variables|Default|SetIsInverted true)
      (Variables|Default|SetCurrentRiftDuration (Variables|Default|GetIntoRiftReverseDuration)))))

(event Custom|CleanupTimers
  (Utilities|Time|ClearandInvalidateTimerbyHandle (Variables|Default|GetIntoRiftIntervalTimer))
  (Utilities|Time|ClearandInvalidateTimerbyHandle (Variables|Default|GetIntoRiftDurationTimer))
  (Variables|Default|SetIsInverted false)
  (Variables|Default|SetCurrentRiftDuration (Variables|Default|GetIntoRiftDuration))
  (Variables|Default|SetVfxActive false))

(event EventEndPlay
  (params (EndPlayReason :type EEndPlayReason%20Enum :default "Destroyed"))
  (CallFunction|CleanupTimers)
  (|CallParentFunction EndPlayReason))

(fn ConstructionScript
  (|CallParentFunction))
