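; Parent Blueprint for Battle Pass S23 Part 02 Q04 "Send AI" quest actors, overriding the inherited state change that follows player interaction.
; It blocks further interaction, adds a success Niagara effect, then shrinks and hides the mesh by timeline or swaps to post-interaction visuals.
; Cosmetic events fire before or after the change by a signed time shift, and authority applies gameplay effects.
; A HasBeenHidden flag feeds a fix for early hiding on completion.
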(event EventBeginPlay
  (|CallParentFunction) ; This node is disabled and will not be called.\nDrag off pins to build functionality.
)

(event Collision|EventActorBeginOverlap
  (params (OtherActor :type Actor%20Object%20Reference))
  (|CallParentFunction OtherActor) ; This node is disabled and will not be called.\nDrag off pins to build functionality.
)

(event EventTick
  (params (DeltaSeconds :type Float%20%28single-precision%29))
  (|CallParentFunction DeltaSeconds) ; This node is disabled and will not be called.\nDrag off pins to build functionality.
)

(event EventChangeStateForPlayer
  (params (FPC :type Fort%20Player%20Controller%20Object%20Reference))
  (bind (_preinteraction_mesh_transform_location _preinteraction_mesh_transform_rotation _preinteraction_mesh_transform_scale) (Math|Transform|BreakTransform (Variables|GetPreInteraction%20Mesh%20Transform_Location)))
  (AddComponent|NODE_AddNiagaraComponent-0 :RelativeTransform (Variables|Default|GetSuccessNiagaraOffset))
  (Variables|Advanced|Debug|SetbOnVisualTestingCooldown true)
  (if (Variables|Advanced|IntervalCheck|GetbUsingDistanceCheck)
    (CallFunction|EventRemoveWatchedActor FPC)
    (else
      _))
  (InteractionState|HandleSetCannotInteract FPC)
  ;[1 0 0] This needs to work on just the interacting player in splitscreen
  (VisualState|VFX&Audio|NoVFX&Audio FPC)
  (if (< (Variables|Design|CosmeticEvent|GetCosmeticTimeShift) 0.0)
    (if (Variables|Design|CosmeticEvent|GetbHideDuringCosmeticEvent)
      (VisualState|Meshes|HiddenVisuals FPC)
      (else
        _))
    (VisualState|VFX&Audio|CosmeticVFX&Audio FPC)
    (CallFunction|SpawnCosmeticActor)
    (Utilities|FlowControl|Delay (- (Math|Float|Absolute(Float) (Variables|Design|CosmeticEvent|GetCosmeticTimeShift)) (Variables|Design|PostInteraction|GetStateChangeDelay)))
    (else
      _))
  (Utilities|FlowControl|Delay (Variables|Design|PostInteraction|GetStateChangeDelay))
  ;[1 1 1] Hiding or Swapping?
  (if (Variables|Design|PostInteraction|GetbHideAfterInteract)
    (CallFunction|FixforEarlyHiddenwithCompletion (Variables|Default|GetHasBeenHidden) FPC)
    (Utilities|FlowControl|Delay (Variables|Default|GetShrinkDelay))
    ;;[1 0 0] What is this? -Macker
    (AddComponent|NODE_AddStaticMeshComponent-0)
    ;;[1 0 0] Why are we using 3 timelines for tthe same thing? They are each a component.
    (bind (_direction _scaledowntime) (Timeline|Timeline PlayFromStart)
      (:Update
        (Transformation|SetWorldScale3D (Variables|Components|GetSM_Quest_PreInteraction) (Math|Vector|Lerp(Vector) _preinteraction_mesh_transform_scale "0.010000,0.010000,0.010000" _scaledowntime)))
      (:Finished
        (VisualState|Meshes|HiddenVisuals FPC)))
    (else
      (VisualState|Meshes|PostInteractionVisuals FPC)))
  ;[1 1 1] Post-Interaction Particles and Sound
  (CallFunction|PlayRewardFeedback FPC)
  (Networking|SwitchHasAuthority
    (:Authority
      ;;[1 1 1] Applying a Gameplay Effect to Player
      (CallFunction|ManagePlayerGEs FPC)
      ;;[1 1 1] Applying Cosmetic Override to Player
      (if (Variables|Design|QuestSetup|GameplayTags,EffectsandCosmetics|GetbShouldApplyCosmeticOverrides)
        (CallFunction|CosmeticOverridestoPlayer FPC)
        (else
          _))
      (VisualState|VFX&Audio|PostInteractionVFX&Audio FPC)
      (if (>= (Variables|Design|CosmeticEvent|GetCosmeticTimeShift) 0.0)
        (Utilities|FlowControl|Delay (Variables|Design|CosmeticEvent|GetCosmeticTimeShift))
        (VisualState|VFX&Audio|CosmeticVFX&Audio FPC)
        (CallFunction|SpawnCosmeticActor)
        (else
          _))
      (CallFunction|ResetForVisualTesting FPC))
    (:Remote
      _)))

(fn FixForEarlyHiddenWithCompletion
  (params (HasBeenHidden :type Boolean) (PlayerController :type Fort%20Player%20Controller%20Object%20Reference))
  (if HasBeenHidden
    (VisualState|Actor|SetActorHidden PlayerController)
    (Variables|Default|SetHasBeenHidden false)
    (Utilities|IsValid (Variables|Design|PostInteraction|GetSM_PostInteraction)
      (:Is%20Valid
        (VisualState|Meshes|SetMeshHidden (Variables|Components|GetSM_Quest_PostInteraction) PlayerController))
      (:Is%20Not%20Valid
        _))
    (VisualState|Meshes|SetMeshHidden (Variables|Components|GetSM_Quest_PreInteraction) PlayerController)))

(fn HiddenVisuals
  (params (PlayerController :type Fort%20Player%20Controller%20Object%20Reference))
  (|CallParentFunction PlayerController)
  (Variables|Default|SetHasBeenHidden true))

(fn ConstructionScript
  (|CallParentFunction))
